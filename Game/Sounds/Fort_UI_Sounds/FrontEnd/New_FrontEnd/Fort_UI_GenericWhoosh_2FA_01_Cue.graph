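# Front-end UI sound cue that plays a two-part generic whoosh, named for a 2FA screen or prompt, in the Fort_UI sound class.
# A mixer layers a medium whoosh, lightly randomized by a modulator, with a second whoosh that a fixed delay starts half a second later.
# A concurrency override allows one instance at a time and blocks new plays rather than cutting off the current one.
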SoundCue /Game/Sounds/Fort_UI_Sounds/FrontEnd/New_FrontEnd/Fort_UI_GenericWhoosh_2FA_01_Cue
root=Mixer_0
nodes [
 {id=Delay_0, children [WavePlayer_5], properties {delayMax=0.5, delayMin=0.5}}
 {id=Mixer_0, children [Modulator_0,Delay_0], properties {inputVolume [1,1]}}
 {id=Modulator_0, children [WavePlayer_1], properties {volumeMin=1}}
 {
  id=WavePlayer_1
  properties {soundWaveAssetPtr=/Game/Sounds/Fort_UI_Sounds/FrontEnd/_GenericToolkit_/Whooshes/Misc_Whooshes/wav/UI_Toolkit_Whoosh_Med_01}
 }
 {
  id=WavePlayer_5
  properties {soundWaveAssetPtr=/Game/Sounds/Fort_UI_Sounds/FrontEnd/_GenericToolkit_/Whooshes/Misc_Whooshes/wav/UI_Toolkit_Whoosh_Med_09}
 }
]
properties {
 bOverrideConcurrency=true
 maxDistance=2097152
 soundClassObject=&Fort_UI
 volumeMultiplier=0.5
 concurrencyOverrides {maxCount=1, resolutionRule=PreventNew, retriggerTime=1}
}
