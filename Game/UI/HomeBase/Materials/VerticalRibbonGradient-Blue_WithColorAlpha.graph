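# Translucent UI material for HomeBase that, going by its name, draws a blue vertical ribbon gradient with color-driven alpha.
# The V output of LinearGradient blends two color parameters, ColoA (spelled that way) and ColorB, defaulting to brighter teal-blue and dark navy.
# A second lerp on the same gradient blends the two parameters' alpha channels, so each color carries its own transparency.
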Material /Game/UI/HomeBase/Materials/VerticalRibbonGradient-Blue_WithColorAlpha
graph {
 properties {
  bUsedWithStaticMesh=false
  blendMode=BLEND_Translucent
  materialDomain=MD_UI
  pixelDepthOffsetMode=PDOM_Legacy
  shadingModel=MSM_Unlit
  usedShadingModels=Unlit
 }
 nodes [
  n1 = VectorParameter() {parameterName=ColoA, defaultValue=<rgba 0.0047 0.3854 0.5521 1>}
  n2 = VectorParameter() {parameterName=ColorB, defaultValue=<rgba 0 0.0319 0.0545 1>}
  n3 = MaterialFunctionCall() {materialFunction=&LinearGradient}
  LinearInterpolate(A=@n1.RGB, B=@n2.RGB, Alpha=@n3.VGradient)
  LinearInterpolate(A=@n1.A, B=@n2.A, Alpha=@n3.VGradient) {constA=1, constB=0.8}
 ]
}
