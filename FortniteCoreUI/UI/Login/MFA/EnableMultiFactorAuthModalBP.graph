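# Login modal widget, based on EnableMultiFactorModal, that urges players to enable two-factor authentication for the Boogie Down emote.
# HandleSetScreenMode switches between an enable button and a console URL with confirmation button, recentering on the active button.
# Activation resets Intro_V2 via reverse playback, then replays it after a brief delay with a whoosh sound.
# NavUp and NavRight handle directional focus; ScaleTitleForCulture's length-based title scaling is unconnected to its entry.
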WidgetBlueprint /FortniteCoreUI/UI/Login/MFA/EnableMultiFactorAuthModalBP
animations [
 Intro {
  binding_count=2
  display_label=""
  length_seconds=1.1159
  object=&self:Intro
  sequence {bindings [
    CT_RewardHeader {
     id=#1
     possessed_class=CommonTextBlock
     tracks [{
       type=MovieScene2DTransformTrack
       sections [{
         type=MovieScene2DTransformSection
         end_frame=18
         start_frame=-1
         channels [
          Translation.X {type=MovieSceneScriptingFloatChannel, thinned_from=5, keys [[0,-500],[4,0],[18,0]]}
          Translation.Y {type=MovieSceneScriptingFloatChannel, keys [[0,0],[14,0],[15,22],[18,0]]}
          Angle {type=MovieSceneScriptingFloatChannel}
          Scale.X {type=MovieSceneScriptingFloatChannel, const [1,3]}
          Scale.Y {type=MovieSceneScriptingFloatChannel, keys [[14,1],[15,0.8],[17,1]]}
          Shear.X {type=MovieSceneScriptingFloatChannel, keys [[-0.3177,34.8808],[3,34.8808],[5,-18],[7,15],[9,0]]}
          Shear.Y {type=MovieSceneScriptingFloatChannel, const [0,1]}
         ]
        }]
      }]
    }
    CT_LimitedTimeHeader {
     id=#2
     possessed_class=CommonTextBlock
     tracks [{
       type=MovieScene2DTransformTrack
       sections [{
         type=MovieScene2DTransformSection
         end_frame=22
         start_frame=0
         channels [
          Translation.X {type=MovieSceneScriptingFloatChannel, const [0,2]}
          Translation.Y {type=MovieSceneScriptingFloatChannel, keys [[0,-191],[12,-191],[15,31],[17,-8.9599],[20,10.5658],[22,0]]}
          Angle {type=MovieSceneScriptingFloatChannel}
          Scale.X {type=MovieSceneScriptingFloatChannel, const [1,1]}
          Scale.Y {type=MovieSceneScriptingFloatChannel, keys [[0,2],[14,2],[15,0.75],[17,1]]}
          Shear.X {type=MovieSceneScriptingFloatChannel}
          Shear.Y {type=MovieSceneScriptingFloatChannel}
         ]
        }]
      }]
    }
   ]}
 }
 ArrowPulse {
  binding_count=1
  display_label=""
  length_seconds=1
  object=&self:ArrowPulse
  sequence {bindings [Image_2 {
     id=#3
     possessed_class=Image
     tracks [{
       type=MovieScene2DTransformTrack
       sections [{
         type=MovieScene2DTransformSection
         end_frame=20
         start_frame=0
         channels [
          Translation.X {type=MovieSceneScriptingFloatChannel, keys [[0,0],[10,50],[20,0]]}
          Translation.Y {type=MovieSceneScriptingFloatChannel}
          Angle {type=MovieSceneScriptingFloatChannel}
          Scale.X {type=MovieSceneScriptingFloatChannel, keys [[0,1],[10,0.8509],[20,1]]}
          Scale.Y {type=MovieSceneScriptingFloatChannel}
          Shear.X {type=MovieSceneScriptingFloatChannel}
          Shear.Y {type=MovieSceneScriptingFloatChannel}
         ]
        }]
      }]
    }]}
 }
 Intro_V2 {
  binding_count=9
  display_label=""
  length_seconds=1.8432
  object=&self:Intro_V2
  sequence {
   bindings [
    Button_Yes {
     id=#4
     possessed_class=IconTextButton_C
     tracks [{
       type=MovieScene2DTransformTrack
       sections [{
         type=MovieScene2DTransformSection
         end_frame=23
         start_frame=17
         channels [
          Translation.X {type=MovieSceneScriptingFloatChannel}
          Translation.Y {type=MovieSceneScriptingFloatChannel}
          Angle {type=MovieSceneScriptingFloatChannel}
          Scale.X {type=MovieSceneScriptingFloatChannel, keys [[17,0.4],[20,1.1],[22,0.9358],[23,1]]}
          Scale.Y {type=MovieSceneScriptingFloatChannel, keys [[17,0.4],[20,1.1],[22,0.9358],[23,1]]}
          Shear.X {type=MovieSceneScriptingFloatChannel}
          Shear.Y {type=MovieSceneScriptingFloatChannel}
         ]
        }]
      }]
    }
    VerticalBox_6 {
     id=#5
     possessed_class=VerticalBox
     tracks [{
       type=MovieSceneFloatTrack
       display_name=Render Opacity
       sections [{
         type=MovieSceneFloatSection
         end_frame=23
         start_frame=0
         channels [None {type=MovieSceneScriptingFloatChannel, keys [[0,0],[16,0],[19,1],[23,1]]}]
        }]
      }]
    }
    Button_OptOut {
     id=#6
     possessed_class=IconTextButton_C
     tracks [{
       type=MovieSceneFloatTrack
       display_name=Render Opacity
       sections [{type=MovieSceneFloatSection, end_frame=29, start_frame=0, channels [None {type=MovieSceneScriptingFloatChannel, keys [[0,0],[26,0],[29,1]]}]}]
      }]
    }
    Button_RemindLater {
     id=#7
     possessed_class=IconTextButton_C
     tracks [{
       type=MovieSceneFloatTrack
       display_name=Render Opacity
       sections [{
         type=MovieSceneFloatSection
         end_frame=29
         start_frame=0
         channels [None {type=MovieSceneScriptingFloatChannel, thinned_from=4, keys [[0,0],[26,0],[29,1]]}]
        }]
      }]
    }
    CT_LimitedTimeHeader {
     id=#8
     possessed_class=CommonTextBlock
     tracks [{
       type=MovieScene2DTransformTrack
       sections [{
         type=MovieScene2DTransformSection
         end_frame=9
         start_frame=0
         channels [
          Translation.X {type=MovieSceneScriptingFloatChannel, keys [[0,-850],[4,0]]}
          Translation.Y {type=MovieSceneScriptingFloatChannel}
          Angle {type=MovieSceneScriptingFloatChannel}
          Scale.X {type=MovieSceneScriptingFloatChannel}
          Scale.Y {type=MovieSceneScriptingFloatChannel}
          Shear.X {type=MovieSceneScriptingFloatChannel, keys [[0,16],[4,16],[5,-10],[7,0]]}
          Shear.Y {type=MovieSceneScriptingFloatChannel}
         ]
        }]
      }]
    }
    CT_RewardHeader {
     id=#9
     possessed_class=CommonTextBlock
     tracks [{
       type=MovieScene2DTransformTrack
       sections [{
         type=MovieScene2DTransformSection
         end_frame=16
         start_frame=0
         channels [
          Translation.X {type=MovieSceneScriptingFloatChannel, const [0,3]}
          Translation.Y {type=MovieSceneScriptingFloatChannel, keys [[0,1000],[7,1000],[10,0]]}
          Angle {type=MovieSceneScriptingFloatChannel}
          Scale.X {type=MovieSceneScriptingFloatChannel, const [1,2]}
          Scale.Y {type=MovieSceneScriptingFloatChannel, keys [[0,1.5],[9,1.5],[10,0.5765],[12,1.15],[16,1]]}
          Shear.X {type=MovieSceneScriptingFloatChannel}
          Shear.Y {type=MovieSceneScriptingFloatChannel}
         ]
        }]
      }]
    }
    Button_EnableMFA {
     id=#10
     possessed_class=IconTextButton_C
     tracks [{
       type=MovieScene2DTransformTrack
       sections [{
         type=MovieScene2DTransformSection
         end_frame=23
         start_frame=17
         channels [
          Translation.X {type=MovieSceneScriptingFloatChannel}
          Translation.Y {type=MovieSceneScriptingFloatChannel}
          Angle {type=MovieSceneScriptingFloatChannel}
          Scale.X {type=MovieSceneScriptingFloatChannel, keys [[17,0.4],[20,1.1],[22,0.9358],[23,1]]}
          Scale.Y {type=MovieSceneScriptingFloatChannel, keys [[17,0.4],[20,1.1],[22,0.9358],[23,1]]}
          Shear.X {type=MovieSceneScriptingFloatChannel}
          Shear.Y {type=MovieSceneScriptingFloatChannel}
         ]
        }]
      }]
    }
    Button_No {
     id=#11
     possessed_class=IconTextButton_C
     tracks [{
       type=MovieSceneFloatTrack
       display_name=Render Opacity
       sections [{
         type=MovieSceneFloatSection
         end_frame=29
         start_frame=0
         channels [None {type=MovieSceneScriptingFloatChannel, thinned_from=4, keys [[0,0],[26,0],[29,1]]}]
        }]
      }]
    }
    VerticalBox_3 {
     id=#12
     possessed_class=VerticalBox
     tracks [{
       type=MovieSceneFloatTrack
       display_name=Render Opacity
       sections [{
         type=MovieSceneFloatSection
         end_frame=23
         start_frame=0
         channels [None {type=MovieSceneScriptingFloatChannel, keys [[0,0],[16,0],[19,1],[23,1]]}]
        }]
      }]
    }
   ]
   tracks [
    {type=MovieSceneEventTrack, sections [{type=MovieSceneEventSection}]}
    {
     type=MovieSceneAudioTrack
     display_name=Audio
     sections [{
       type=MovieSceneAudioSection
       end_frame=17
       start_frame=0
       channels [
        Volume {type=MovieSceneScriptingFloatChannel, default=1}
        Pitch {type=MovieSceneScriptingFloatChannel, default=1}
        AttachActor {type=MovieSceneScriptingActorReferenceChannel, default="<Struct 'MovieSceneObjectBindingID' (0x4676ea805a0) {}>"}
       ]
      }]
    }
    {
     type=MovieSceneAudioTrack
     display_name=Audio
     sections [{
       type=MovieSceneAudioSection
       end_frame=36
       start_frame=13
       channels [
        Volume {type=MovieSceneScriptingFloatChannel, default=1}
        Pitch {type=MovieSceneScriptingFloatChannel, default=1}
        AttachActor {type=MovieSceneScriptingActorReferenceChannel, default="<Struct 'MovieSceneObjectBindingID' (0x4676ea89220) {}>"}
       ]
      }]
    }
   ]
  }
 }
]
graphs [
 AnimationFullyCompleteBP {nodes [
   AnimationFullyCompleteBP() {is=title K2Node_FunctionEntry}
   > HandleSetScreenMode()
   > AnimationsComplete(self:Enable Multi Factor Modal Object Reference)
  ]}
 EventGraph {nodes [
   HandleHeaderText() {is=title K2Node_CustomEvent}
   > "SetText (Text)"(
    self=Get CT_LimitedTimeHeader() -> (CT_LimitedTimeHeader:Common Text Object Reference)
    InText=TextToUpper!(InText:"Text (by ref)"=MakeLiteralText(Value="LOCTEXT(\"ENABLE 2FA\")")) -> (ReturnValue:Text)
   )
   Event PreConstruct()
   > HandleHeaderText() {is=title K2Node_CallFunction}
   > ScaleTitleForCulture()
   > SetBrushFromSoftTexture!(
    execute
    self:Image Object Reference=Get lazyImage() -> (lazyImage:Fort Lazy Image Object Reference)
    SoftTexture:Texture 2D Soft Object Reference=&/BoogieDown_Large
    bMatchSize:bool=false
   ) -> (then)
   Event Construct()
   > Branch(Condition=IsMobileGame())
   > SetStyle!(
    execute
    self:Common Text Object Reference=Get CT_LimitedTimeHeader()
    InStyle:Common Text Style Class Reference=Get MobileLimitedTimeHeaderStyle() -> (MobileLimitedTimeHeaderStyle:Common Text Style Class Reference)
   ) -> (then)
   Event On Activated()
   > HandleHeaderText()
   n1 = Get Intro_V2() -> (Intro_V2:Widget Animation Object Reference)
   > PlayAnimation(InAnimation=@n1, StartAtTime=GetEndTime(self=@n1), PlayMode=Reverse)
   > Delay(Duration=Get HeartbeatDelayIntroAnimation() -> (HeartbeatDelayIntroAnimation:double))
   > PlayAnimation(InAnimation=Get Intro_V2())
   > ScaleTitleForCulture()
   > SpawnSound2D!(
    execute
    Sound:Sound Base Object Reference=/Game/Sounds/Fort_UI_Sounds/FrontEnd/New_FrontEnd/Fort_UI_GenericWhoosh_2FA_01_Cue
    VolumeMultiplier:float=1
    PitchMultiplier:float=1
    StartTime:float=0
    ConcurrencySettings:Sound Concurrency Object Reference
    bPersistAcrossLevelTransition:bool=false
    bAutoDestroy:bool=true
   ) -> (then, ReturnValue:Audio Component Object Reference)
   K2Node_Event_4 = OnIncentivizedSet!() -> (OutputDelegate, then, bIncentivized:bool=~) {is=title K2Node_CustomEvent}
   > Set bHasReward(bHasReward:bool=false) -> (Output_Get:bool)
   n2 = Event OnSetExitButtonText!() -> (OutputDelegate, then, NewButtonText:"Text (by ref)") {is=title K2Node_Event}
   > n3 = Cast To IconTextButton!(execute, Object=Get Button_RemindLater() -> (Button_RemindLater:Common Button Legacy Object Reference)) -> (then, CastFailed, AsIcon Text Button:Icon Text Button Object Reference) {is=title K2Node_DynamicCast}
   > Set Text!(execute, self:Icon Text Button Object Reference=@n3, ButtonText:Text=@n2.NewButtonText) -> (then) {is=title K2Node_CallFunction}
   n4 = Event OnConsoleDisplayURLProvided!() -> (OutputDelegate, then, UniquePlayerURLText:"Text (by ref)") {is=title K2Node_Event}
   > "SetText (Text)"(self=Get ConsoleTextBlock() -> (ConsoleTextBlock:Common Text Object Reference), InText=@n4.UniquePlayerURLText)
   Event OnInputModeChanged!() -> (OutputDelegate, then, bUsingGamepad:bool) {is=title K2Node_Event}
   > HandleSetScreenMode()
   Event OnSetScreenConfiguration!() -> (OutputDelegate, then, bIsConsole:bool) {is=title K2Node_Event}
   > HandleSetScreenMode()
   Get Intro()
  ]}
 HandleSetScreenMode {nodes [
   HandleSetScreenMode() {is=title K2Node_FunctionEntry}
   > n1 = Branch(Condition=Get bShouldShowConsoleVersion() -> (bShouldShowConsoleVersion:bool))
   n2 = Get EnableButtonSwitcher() -> (EnableButtonSwitcher:Widget Switcher Object Reference)
   > SetActiveWidgetIndex(self=@n2, Index=1)
   > SetVisibility(self=Get Button_ConsoleMFAEnabled() -> (Button_ConsoleMFAEnabled:Common Button Legacy Object Reference), InVisibility=SelfHitTestInvisible)
   > CenterWidget!(execute, ToCenter:Widget Object Reference=Get Button_ConsoleMFAEnabled()) -> (then)
   > SetVisibility(self=Get CT_TakenToWebsite() -> (CT_TakenToWebsite:Common Text Object Reference), InVisibility=Visible)
   SetActiveWidgetIndex(execute=@n1.else, self=@n2, Index=0)
   > SetVisibility(self=Get Button_ConsoleMFAEnabled(), InVisibility=Collapsed)
   > CenterWidget(ToCenter=Get Button_EnableMFA() -> (Button_EnableMFA:Common Button Legacy Object Reference))
   > SetVisibility(self=Get CT_TakenToWebsite(), InVisibility=Hidden)
  ]}
 NavRight {nodes [
   NavRight() {is=title K2Node_FunctionEntry}
   > n1 = Branch(Condition=HasAnyUserFocus!(self:Widget Object Reference=Get Button_RemindLater()) -> (ReturnValue:bool))
   > Branch(Condition=GetIsEnabled!(self:Widget Object Reference=Get Button_OptOut() -> (Button_OptOut:Common Button Legacy Object Reference)) -> (ReturnValue:bool))
   > CenterWidget(ToCenter=Get Button_OptOut())
   CenterWidget(execute=@n1.else, ToCenter=Get Button_RemindLater())
  ]}
 NavUp {nodes [
   NavUp() {is=title K2Node_FunctionEntry}
   n1 = Get EnableButtonSwitcher()
   > n2 = Branch(Condition="Equal (Integer)"(A=GetActiveWidgetIndex!(self:Widget Switcher Object Reference=@n1) -> (ReturnValue:int=0), B=0) {is=title K2Node_CallFunction})
   n3 = Get Button_EnableMFA()
   > Branch(Condition=GetIsEnabled(self=@n3))
   > CenterWidget(ToCenter=@n3)
   n4 = Get Button_ConsoleMFAEnabled()
   Branch(execute=@n2.else, Condition=GetIsEnabled(self=@n4))
   > CenterWidget(ToCenter=@n4)
  ]}
 ScaleTitleForCulture {nodes [
   n1 = Get CT_LimitedTimeHeader()
   n2 = Len(S="To String (Text)"(InText="GetText (Text)"!(self:Text Object Reference=@n1) -> (ReturnValue:Text)))
   n3 = "Max (Float)"(
    A=float / float(A="To Float (Integer)"(InInt=@n2), B=Get ScaleCharThreshold() -> (ScaleCharThreshold:double)) {is=title K2Node_CallFunction}
    B=1
   )
   SetUserSpecifiedScale!(
    execute
    self:Scale Box Object Reference=Get ScaleBox_TitleHeader() -> (ScaleBox_TitleHeader:Scale Box Object Reference)
    InUserSpecifiedScale:float="Clamp (Float)"(Value=float / float(A=1, B=@n3), Min=0.3, Max=1)
   ) -> (then)
   ScaleTitleForCulture() {is=title K2Node_FunctionEntry}
  ]}
]
variables [
 bHasReward {type=boolean}
 HeartbeatDelayIntroAnimation {type=number, default=0.15}
 MobileLimitedTimeHeaderStyle {type=Class, default=/FortUILibrary/Styles/Typography/Foundation/TextStyle-Header-XXL.TextStyle-Header-XXL_C}
]
widgets {
 info {
  inheritedWidgetCount=5
  namedSlotCount=1
  parentClass=/Script/FortniteUI.EnableMultiFactorModal
  rootWidgetClass=&Border
  widgetCount=26
 }
 widgets [
  Border_0 {
   widgetClassPath=&Border
   widget=Border_0
   properties {background {
     bIsDynamicallyLoaded=false
     drawAs=Image
     imageType=NoImage
     mirroring=NoMirror
     resourceName=None
     resourceObject=/Game/UI/HomeBase/Materials/VerticalRibbonGradient-Blue_WithColorAlpha
     tiling=NoTile
     imageSize=<32 32>
     margin=<ltrb 0 0 0 0>
     outlineSettings {
      bUseBrushTransparency=false
      roundingType=HalfHeightRadius
      width=0
      color {colorUseRule=UseColor_Specified, specifiedColor=<rgba 0 0 0 0>}
      cornerRadii=<0 0 0 0>
     }
     tintColor {colorUseRule=UseColor_Specified, specifiedColor=<rgba 1 1 1 1>}
     uVRegion {bIsValid=false, max=<0 0>, min=<0 0>}
    }}
  }
  Overlay_3 {widgetClassPath=&Overlay, parent=Border_0, widget=Overlay_3, slot {class=BorderSlot}}
  Overlay_1 {
   widgetClassPath=&Overlay
   parent=Overlay_3
   widget=Overlay_1
   slot {class=OverlaySlot, properties {horizontalAlignment=HAlign_Right, verticalAlignment=VAlign_Bottom, padding=<ltrb 0 0 0 50>}}
  }
  lazyImage {
   widgetClassPath=&FortLazyImage
   bIsVariable=true
   parent=Overlay_1
   widget=lazyImage
   properties {
    brush {
     bIsDynamicallyLoaded=false
     drawAs=Image
     imageType=NoImage
     mirroring=NoMirror
     resourceName=None
     resourceObject=null
     tiling=NoTile
     imageSize=<816 816>
     margin=<ltrb 0 0 0 0>
     outlineSettings {
      bUseBrushTransparency=false
      roundingType=HalfHeightRadius
      width=0
      color {colorUseRule=UseColor_Specified, specifiedColor=<rgba 0 0 0 0>}
      cornerRadii=<0 0 0 0>
     }
     tintColor {colorUseRule=UseColor_Specified, specifiedColor=<rgba 1 1 1 1>}
     uVRegion {bIsValid=false, max=<0 0>, min=<0 0>}
    }
    renderTransform {angle=0, scale=<1.15 1.15>, shear=<0 0>, translation=<0 0>}
   }
   slot {class=OverlaySlot, properties {padding=<ltrb 0 0 0 50>}}
  }
  SafeZone {
   widgetClassPath=&SafeZone
   bIsVariable=true
   parent=Overlay_3
   widget=SafeZone
   properties {padBottom=false, padTop=false, visibility=Visible}
   slot {class=OverlaySlot, properties {horizontalAlignment=HAlign_Fill, verticalAlignment=VAlign_Fill}}
  }
  Overlay_PreventLanguageButtonPushing {
   widgetClassPath=&Overlay
   parent=SafeZone
   widget=Overlay_PreventLanguageButtonPushing
   slot {class=SafeZoneSlot, properties {padding=<ltrb 80 40 80 40>}}
  }
  HorizontalBox_0 {
   widgetClassPath=&HorizontalBox
   parent=Overlay_PreventLanguageButtonPushing
   widget=HorizontalBox_0
   slot {class=OverlaySlot, properties {horizontalAlignment=HAlign_Fill, verticalAlignment=VAlign_Fill}}
  }
  VerticalBox_2 {
   widgetClassPath=&VerticalBox
   parent=HorizontalBox_0
   widget=VerticalBox_2
   slot {class=HorizontalBoxSlot, properties {padding=<ltrb 50 0 50 0>}}
  }
  SizeBox_0 {
   widgetClassPath=&SizeBox
   parent=VerticalBox_2
   widget=SizeBox_0
   properties {bOverride_MaxDesiredWidth=true, maxDesiredWidth=900}
   slot {class=VerticalBoxSlot, properties {horizontalAlignment=HAlign_Left, verticalAlignment=VAlign_Top}}
  }
  ScaleBox_TitleHeader {widgetClassPath=&ScaleBox, bIsVariable=true, parent=SizeBox_0, widget=ScaleBox_TitleHeader, slot {class=SizeBoxSlot}}
  CT_LimitedTimeHeader {
   widgetClassPath=&CommonTextBlock
   bIsVariable=true
   parent=ScaleBox_TitleHeader
   widget=CT_LimitedTimeHeader
   properties {
    style=/FortUILibrary/Styles/Typography/Foundation/TextStyle-Header-XXXL.TextStyle-Header-XXXL_C
    text=ENABLE 2FA TEST
    font {
     bForceMonospaced=false
     bMaterialIsStencil=false
     fontMaterial=null
     fontObject=&Burbank
     letterSpacing=0
     monospacedWidth=1
     size=175
     skewAmount=0
     typefaceFontName=Burbank-Black
     outlineSettings {
      bApplyOutlineToDropShadows=true
      bMiteredCorners=false
      bSeparateFillAlpha=false
      outlineMaterial=null
      outlineSize=3
      outlineColor=<rgba 0 0 0 0.1>
     }
    }
    renderTransformPivot=<0.5 1>
    shadowOffset=<0 0>
   }
   slot {class=ScaleBoxSlot}
  }
  VerticalBox_0 {
   widgetClassPath=&VerticalBox
   parent=VerticalBox_2
   widget=VerticalBox_0
   slot {class=VerticalBoxSlot, properties {size {sizeRule=Fill, value=1}}}
  }
  VerticalBox_3 {
   widgetClassPath=&VerticalBox
   parent=VerticalBox_0
   widget=VerticalBox_3
   slot {class=VerticalBoxSlot, properties {horizontalAlignment=HAlign_Left, size {sizeRule=Fill, value=1}}}
  }
  CT_MFA_Description {
   widgetClassPath=&CommonTextBlock
   bInherited=true
   bIsVariable=true
   parent=VerticalBox_3
   widget=CT_MFA_Description
   properties {
    style=/FortUILibrary/Styles/Typography/Foundation/TextStyle-Base-M.TextStyle-Base-M_C
    text="Two-Factor Authentication (2FA) protects your account. Enable 2FA to get the Boogie Down emote!"
    wrapTextAt=725
    font {
     bForceMonospaced=false
     bMaterialIsStencil=false
     fontMaterial=null
     fontObject=&NotoSans
     letterSpacing=0
     monospacedWidth=1
     size=24
     skewAmount=0
     typefaceFontName=NotoSans-Regular
     outlineSettings {
      bApplyOutlineToDropShadows=false
      bMiteredCorners=false
      bSeparateFillAlpha=false
      outlineMaterial=null
      outlineSize=0
      outlineColor=<rgba 0 0 0 1>
     }
    }
    shadowOffset=<0 0>
   }
   slot {class=VerticalBoxSlot, properties {padding=<ltrb 0 60 0 40>}}
  }
  VerticalBox_5 {widgetClassPath=&VerticalBox, parent=VerticalBox_3, widget=VerticalBox_5, slot {class=VerticalBoxSlot}}
  EnableButtonSwitcher {
   widgetClassPath=&WidgetSwitcher
   bIsVariable=true
   parent=VerticalBox_5
   widget=EnableButtonSwitcher
   slot {class=VerticalBoxSlot, properties {horizontalAlignment=HAlign_Left}}
  }
  Button_EnableMFA {
   widgetClassPath=&IconTextButton
   bInherited=true
   bIsVariable=true
   parent=EnableButtonSwitcher
   widget=Button_EnableMFA
   properties {
    buttonText=Enable 2FA Now
    contentAlignment=HAlign_Center
    minHeight=125
    minWidth=550
    navigation=&self:WidgetTree.Button_EnableMFA.WidgetNavigation_0
    onButtonBaseClicked=(&self:WidgetTree.Button_EnableMFA.HandleOnButtonClicked)
    onButtonBaseDoubleClicked=(&self:WidgetTree.Button_EnableMFA.HandleOnButtonDoubleClicked)
    onButtonBaseHovered=(&self:WidgetTree.Button_EnableMFA.HandleOnButtonHovered)
    onButtonBaseUnhovered=(&self:WidgetTree.Button_EnableMFA.HandleOnButtonUnhovered)
    onSelectedChangedBase=(&self:WidgetTree.Button_EnableMFA.HandleOnSelectedChanged)
    style=/FortUILibrary/Foundation/Buttons/ButtonStyles/Skew/ButtonStyle-Skew_Desirable.ButtonStyle-Skew_Desirable_C
    iconBrush {
     bIsDynamicallyLoaded=false
     drawAs=Image
     imageType=NoImage
     mirroring=NoMirror
     resourceName=None
     resourceObject=/FortniteCoreUI/UI/Login/Art/M-Icon-ExternalLink-64
     tiling=NoTile
     imageSize=<0 0>
     margin=<ltrb 0 0 0 0>
     outlineSettings {
      bUseBrushTransparency=false
      roundingType=HalfHeightRadius
      width=0
      color {colorUseRule=UseColor_Specified, specifiedColor=<rgba 0 0 0 0>}
      cornerRadii=<0 0 0 1>
     }
     tintColor {colorUseRule=UseColor_Specified, specifiedColor=<rgba 1 1 1 1>}
     uVRegion {bIsValid=false, max=<0 0>, min=<0 0>}
    }
   }
   slot {class=WidgetSwitcherSlot}
  }
  Image_GoToWebsite {
   widgetClassPath=&Image
   bIsVariable=true
   namedSlotHost=Button_EnableMFA
   widget=Image_GoToWebsite
   properties {
    brush {
     bIsDynamicallyLoaded=false
     drawAs=Image
     imageType=NoImage
     mirroring=NoMirror
     resourceName=None
     resourceObject=&M-Icon-ExternalLink-64
     tiling=NoTile
     imageSize=<92 92>
     margin=<ltrb 0 0 0 0>
     outlineSettings {
      bUseBrushTransparency=false
      roundingType=HalfHeightRadius
      width=0
      color {colorUseRule=UseColor_Specified, specifiedColor=<rgba 0 0 0 0>}
      cornerRadii=<0 0 0 0>
     }
     tintColor {colorUseRule=UseColor_Specified, specifiedColor=<rgba 0.0052 0.0052 0.0056 1>}
     uVRegion {bIsValid=false, max=<0 0>, min=<0 0>}
    }
    renderTransform {angle=0, scale=<1 1>, shear=<0 0>, translation=<0 -5>}
   }
  }
  ConsoleTextBlock {
   widgetClassPath=&CommonTextBlock
   bIsVariable=true
   parent=EnableButtonSwitcher
   widget=ConsoleTextBlock
   properties {
    text=https://epicgames.com/2FA
    wrapTextAt=725
    colorAndOpacity {colorUseRule=UseColor_Specified, specifiedColor=<rgba 1 0.9789 0 1>}
    font {
     bForceMonospaced=false
     bMaterialIsStencil=false
     fontMaterial=null
     fontObject=&BurbankSmall
     letterSpacing=0
     monospacedWidth=1
     size=38
     skewAmount=0
     typefaceFontName=NotoSans-Regular
     outlineSettings {
      bApplyOutlineToDropShadows=false
      bMiteredCorners=false
      bSeparateFillAlpha=false
      outlineMaterial=null
      outlineSize=0
      outlineColor=<rgba 0 0 0 1>
     }
    }
    shadowColorAndOpacity=<rgba 0.0052 0.0052 0.0056 0.52>
    shadowOffset=<3 3>
   }
   slot {class=WidgetSwitcherSlot, properties {verticalAlignment=VAlign_Top}}
  }
  CT_TakenToWebsite {
   widgetClassPath=&CommonTextBlock
   bIsVariable=true
   parent=VerticalBox_5
   widget=CT_TakenToWebsite
   properties {
    style=/FortUILibrary/Styles/Typography/Foundation/TextStyle-Base-XS.TextStyle-Base-XS_C
    text=Note - You'll need to visit in a web browser.
    visibility=Visible
    font {
     bForceMonospaced=false
     bMaterialIsStencil=false
     fontMaterial=null
     fontObject=&NotoSans
     letterSpacing=0
     monospacedWidth=1
     size=16
     skewAmount=0
     typefaceFontName=NotoSans-Regular
     outlineSettings {
      bApplyOutlineToDropShadows=false
      bMiteredCorners=false
      bSeparateFillAlpha=false
      outlineMaterial=null
      outlineSize=0
      outlineColor=<rgba 0 0 0 1>
     }
    }
    shadowOffset=<0 0>
   }
   slot {class=VerticalBoxSlot, properties {padding=<ltrb 5 0 0 0>}}
  }
  Button_ConsoleMFAEnabled {
   widgetClassPath=&IconTextButton
   bInherited=true
   bIsVariable=true
   parent=VerticalBox_5
   widget=Button_ConsoleMFAEnabled
   properties {
    buttonText=I've Enabled 2FA
    contentAlignment=HAlign_Center
    minHeight=85
    minWidth=300
    navigation=&self:WidgetTree.Button_ConsoleMFAEnabled.WidgetNavigation_0
    onButtonBaseClicked=(&self:WidgetTree.Button_ConsoleMFAEnabled.HandleOnButtonClicked)
    onButtonBaseDoubleClicked=(&self:WidgetTree.Button_ConsoleMFAEnabled.HandleOnButtonDoubleClicked)
    onButtonBaseHovered=(&self:WidgetTree.Button_ConsoleMFAEnabled.HandleOnButtonHovered)
    onButtonBaseUnhovered=(&self:WidgetTree.Button_ConsoleMFAEnabled.HandleOnButtonUnhovered)
    onSelectedChangedBase=(&self:WidgetTree.Button_ConsoleMFAEnabled.HandleOnSelectedChanged)
    style=&ButtonStyle-Skew_Desirable_C
    iconBrush {
     bIsDynamicallyLoaded=false
     drawAs=Image
     imageType=NoImage
     mirroring=NoMirror
     resourceName=None
     resourceObject=&M-Icon-ExternalLink-64
     tiling=NoTile
     imageSize=<0 0>
     margin=<ltrb 0 0 0 0>
     outlineSettings {
      bUseBrushTransparency=false
      roundingType=HalfHeightRadius
      width=0
      color {colorUseRule=UseColor_Specified, specifiedColor=<rgba 0 0 0 0>}
      cornerRadii=<0 0 0 1>
     }
     tintColor {colorUseRule=UseColor_Specified, specifiedColor=<rgba 1 1 1 1>}
     uVRegion {bIsValid=false, max=<0 0>, min=<0 0>}
    }
   }
   slot {class=VerticalBoxSlot, properties {padding=<ltrb 0 0 100 0>}}
  }
  HorizontalBox_1 {
   widgetClassPath=&HorizontalBox
   parent=VerticalBox_3
   widget=HorizontalBox_1
   slot {class=VerticalBoxSlot, properties {horizontalAlignment=HAlign_Left, verticalAlignment=VAlign_Bottom, size {sizeRule=Fill, value=1}}}
  }
  Button_RemindLater {
   widgetClassPath=&IconTextButton
   bInherited=true
   bIsVariable=true
   parent=HorizontalBox_1
   widget=Button_RemindLater
   properties {
    buttonText=Remind Me Later
    contentAlignment=HAlign_Center
    minHeight=85
    minWidth=300
    navigation=&self:WidgetTree.Button_RemindLater.WidgetNavigation_0
    onButtonBaseClicked=(&self:WidgetTree.Button_RemindLater.HandleOnButtonClicked)
    onButtonBaseDoubleClicked=(&self:WidgetTree.Button_RemindLater.HandleOnButtonDoubleClicked)
    onButtonBaseHovered=(&self:WidgetTree.Button_RemindLater.HandleOnButtonHovered)
    onButtonBaseUnhovered=(&self:WidgetTree.Button_RemindLater.HandleOnButtonUnhovered)
    onSelectedChangedBase=(&self:WidgetTree.Button_RemindLater.HandleOnSelectedChanged)
    style=/FortUILibrary/Foundation/Buttons/ButtonStyles/Skew/ButtonStyle-Skew_LessDesirable.ButtonStyle-Skew_LessDesirable_C
    pressedSlateSoundOverride {resourceObject=/Game/Sounds/Fort_UI_Sounds/FrontEnd/ItemShop/Fortnite_UI_Button_Pressed_Cue}
   }
   slot {class=HorizontalBoxSlot}
  }
  VBox_OptOut {
   widgetClassPath=&VerticalBox
   parent=Overlay_PreventLanguageButtonPushing
   widget=VBox_OptOut
   slot {class=OverlaySlot, properties {horizontalAlignment=HAlign_Right, verticalAlignment=VAlign_Bottom, padding=<ltrb 0 0 60 0>}}
  }
  ItemInfoWidget_FromSocialImport {
   widgetClassPath=/Game/Athena/UI/Season/ItemInfoWidget
   bIsVariable=true
   parent=VBox_OptOut
   widget=ItemInfoWidget_FromSocialImport
   slot {class=VerticalBoxSlot}
  }
  Button_OptOut {
   widgetClassPath=&IconTextButton
   bInherited=true
   bIsVariable=true
   parent=VBox_OptOut
   widget=Button_OptOut
   properties {
    buttonText=Never Show Again
    minHeight=85
    navigation=&self:WidgetTree.Button_OptOut.WidgetNavigation_1
    onButtonBaseClicked=(&self:WidgetTree.Button_OptOut.HandleOnButtonClicked)
    onButtonBaseDoubleClicked=(&self:WidgetTree.Button_OptOut.HandleOnButtonDoubleClicked)
    onButtonBaseHovered=(&self:WidgetTree.Button_OptOut.HandleOnButtonHovered)
    onButtonBaseUnhovered=(&self:WidgetTree.Button_OptOut.HandleOnButtonUnhovered)
    onSelectedChangedBase=(&self:WidgetTree.Button_OptOut.HandleOnSelectedChanged)
    style=&ButtonStyle-Skew_LessDesirable_C
    pressedSlateSoundOverride {resourceObject=&Fortnite_UI_Button_Pressed_Cue}
   }
   slot {class=VerticalBoxSlot, properties {horizontalAlignment=HAlign_Left, padding=<ltrb 0 20 0 0>}}
  }
 ]
}
